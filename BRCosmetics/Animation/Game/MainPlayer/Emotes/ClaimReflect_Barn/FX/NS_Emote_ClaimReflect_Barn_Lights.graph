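# Niagara system for the ClaimReflect_Barn emote's blue stage lights, which pulse with audio-sync values and are tinted by time of day.
# Spotlight_Flare puts nine angled mesh beams with glow sprites on sockets 0–8, scaling and fading with System.AudioSyncAlpha.
# SpotLight adds a flickering-radius light at socket 9, where Smoke_Ground emits a faint ring of drifting, audio-modulated smoke.
# Two disabled debug emitters display audio reactivity and looped age.
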NiagaraSystem /BRCosmetics/Animation/Game/MainPlayer/Emotes/ClaimReflect_Barn/FX/NS_Emote_ClaimReflect_Barn_Lights
dependencies {
 usedDataInterfaces [
  &NiagaraDataInterfaceSpriteRendererInfo
  &NiagaraDataInterfaceCurve
  &NiagaraDataInterfaceSocketReader
  &NiagaraDataInterfaceColorCurve
 ]
 usedDynamicInputs [&ColorFromCurve, /Niagara/DynamicInputs/Lerp/SmoothLerpOverTime_Float]
 usedModules [
  &self:SystemSpawnScript.SystemScriptSource.SystemScriptGraph.NiagaraNodeAssignment_0.SetVariables_#1
  &self:SystemSpawnScript.SystemScriptSource.SystemScriptGraph.NiagaraNodeAssignment_3.SetVariables_#2
  &SystemState
  &self:SystemSpawnScript.SystemScriptSource.SystemScriptGraph.NiagaraNodeAssignment_1.SetVariables_#3
  &self:SystemSpawnScript.SystemScriptSource.SystemScriptGraph.NiagaraNodeAssignment_4.SetVariables_#4
  /Niagara/Noise/Waveform
  &self:SystemSpawnScript.SystemScriptSource.SystemScriptGraph.NiagaraNodeAssignment_7.SetVariables_#5
  &self:SystemSpawnScript.SystemScriptSource.SystemScriptGraph.NiagaraNodeAssignment_8.SetVariables_#6
  &self:SystemSpawnScript.SystemScriptSource.SystemScriptGraph.NiagaraNodeAssignment_2.SetVariables_#7
 ]
}
emitters {
 NE_Debug_Float002 {
  bEnabled=false
  emitterSpawnScript [/SetVariables(Emitter.Digits=2, Emitter.DisplayScalar=true) {isSetParameters=true, module=SetVariables_#8}]
  emitterUpdateScript [
   &EmitterState(Life Cycle Mode=::NewEnumerator1, Loop Behavior=::NewEnumerator1, Loop Duration Mode=::NewEnumerator1, UseLoopDelay=false)
   &SpawnBurst_Instantaneous(Use Spawn Probability=false, UseLoopCountLimit=false)
  ]
  particleSpawnScript [&InitializeParticle(
    Color Mode=::NewEnumerator0
    Lifetime=$Emitter.CurrentLoopDuration
    Position Offset=<30 0 0>
    Sprite Size Mode=::NewEnumerator3
    Uniform Sprite Size=25
    UsePositionOffset=true
    Write Material Random=true
   )]
  particleUpdateScript [
   &ParticleState(Kill Particles When Lifetime Has Elapsed=false)
   &SocketLocation(
    Orientation Sampling=::NewEnumerator1
    SamplePosOffset=true
    Sampled Position Coordinate Space:ENiagaraCoordinateSpace=::World
    Sampled Position Offset=<30 0 50>
    Sampled Position Offset Coordinate Space:ENiagaraSkelMeshTransforms=::NewEnumerator1
    Socket Index=9
    Socket Reader=$System.Socket Reader
    Socket Sampling Mode=::NewEnumerator3
    VelocitySpeedLimit=false
   ) {module=SocketLocation001}
   &Color(Color=/Niagara/DynamicInputs/TypeConversions/MakeLinearColorFromFloat(Value:NiagaraFloat=$System.AudioReactivity))
  ]
  rendererClasses [&NiagaraSpriteRendererProperties]
  renderers [{rendererClass=&NiagaraSpriteRendererProperties}]
 }
 NE_Debug_Float003 {
  bEnabled=false
  emitterSpawnScript [/SetVariables(Emitter.Digits=2, Emitter.DisplayScalar=true) {isSetParameters=true, module=SetVariables_#8}]
  emitterUpdateScript [
   &EmitterState(Life Cycle Mode=::NewEnumerator1, Loop Behavior=::NewEnumerator1, Loop Duration Mode=::NewEnumerator1, UseLoopDelay=false)
   &SpawnBurst_Instantaneous(Use Spawn Probability=false, UseLoopCountLimit=false)
  ]
  particleSpawnScript [&InitializeParticle(
    Color Mode=::NewEnumerator0
    Lifetime=$Emitter.CurrentLoopDuration
    Position Offset=<0 -40 40>
    Sprite Size Mode=::NewEnumerator3
    Uniform Sprite Size=25
    UsePositionOffset=true
    Write Material Random=true
   )]
  particleUpdateScript [
   &ParticleState(Kill Particles When Lifetime Has Elapsed=false)
   &SocketLocation(
    Orientation Sampling=::NewEnumerator1
    SamplePosOffset=true
    Sampled Position Coordinate Space:ENiagaraCoordinateSpace=::World
    Sampled Position Offset=<-30 0 50>
    Sampled Position Offset Coordinate Space:ENiagaraSkelMeshTransforms=::NewEnumerator1
    Socket Index=9
    Socket Reader=$System.Socket Reader
    Socket Sampling Mode=::NewEnumerator3
    VelocitySpeedLimit=false
   ) {module=SocketLocation001}
   &Color(Color=&MakeLinearColorFromFloat(Value:NiagaraFloat=$System.LoopedAge))
  ]
  rendererClasses [&NiagaraSpriteRendererProperties]
  renderers [{rendererClass=&NiagaraSpriteRendererProperties}]
 }
 Smoke_Ground {
  emitterUpdateScript [
   &EmitterState(Life Cycle Mode=::NewEnumerator1, Loop Behavior=::NewEnumerator1, Loop Duration Mode=::NewEnumerator1, UseLoopDelay=false)
   &SpawnRate(SpawnRate=12, Use Spawn Probability=false)
  ]
  particleSpawnScript [
   &InitializeParticle(
    Lifetime Min=1.25
    Lifetime Mode=::NewEnumerator1
    Sprite Rotation Mode=::NewEnumerator2
    Sprite Size Mode=::NewEnumerator1
    Sprite UV Mode=::NewEnumerator3
    Uniform Sprite Size Max=30
    Uniform Sprite Size Min=23
    UsePositionOffset=false
    Write Material Random=true
   )
   &SocketLocation(SamplePosOffset=false, Socket Index=9, Socket Reader=$System.Socket Reader, Socket Sampling Mode=::NewEnumerator3, VelocitySpeedLimit=false) {module=SocketLocation001}
   &ShapeLocation(
    MaskSpawnGroup=false
    Override Seed=false
    Ring Radius=12
    Rotation Axis=<0 1 0>
    Rotation Mode=::NewEnumerator1
    Shape Primitive=::NewEnumerator4
    UseOwnerScale=true
   )
   &SubUVAnimation(
    End Frame Range Override=4
    Evaluation Type=::NewEnumerator0
    "Sprite Renderer "=DataInterface() {propertyValues {SpriteRenderer=&self:Smoke_Ground_0.NiagaraSpriteRendererProperties_1}}
    SubUV Animation Mode:ENiagara_SubUVLookupModeV2=::NewEnumerator2
    UseEndFrame=true
    UseStartFrame=true
   )
   &AddVelocity(
    Module.Use Falloff Distance=false
    Override Seed=false
    Velocity Mode=::NewEnumerator1
    Velocity Speed=&RandomRangeFloat(Minimum=5, Maximum=10)
   )
  ]
  particleUpdateScript [
   &ParticleState()
   &SpriteRotationRate(Rotation Rate=&RandomRangeFloat(Minimum=5, Maximum=50))
   &InheritSourceMovement(
    Applied Force Scale=&VectorFromFloat(Value=0.5)
    Applied Position Falloff Curve=DataInterface() {propertyValues {Curve {
       defaultValue=1
       keys [{interpMode=RCIM_Cubic, tangentMode=RCTM_User, time=0, value=0}, {interpMode=RCIM_Cubic, tangentMode=RCTM_User, time=1, value=1}]
      }}}
    Applied Position Falloff Distance=33
    Applied Position Scale=&VectorFromFloat(Value=1)
    ApplyForce=true
    ApplyForceFalloff=false
    ApplyPosition=true
    ApplyPositionFalloff=true
    ApplyPositionFalloffCurve=true
    Module.LimitForceSpeed=false
   )
   &Drag(Drag=2, Use Linear Drag=true, Use Rotational Drag=true)
   &SolveForcesAndVelocity(Clamp Velocity=false, Limit Acceleration=false)
   &ScaleSpriteSize(Uniform Curve Sprite Scale=DataInterface() {propertyValues {Curve {keys [{arriveTangent=4.1286, interpMode=RCIM_Cubic, leaveTangent=4.1286, tangentMode=RCTM_User, time=0, value=0.5}, <at 1 1.25 RCIM_Cubic>]}}})
   &DynamicMaterialParameters(
    Index 0 Param 1=&RandomRangeFloat(Minimum=0, Maximum=111)
    Index 0 Param 4=0.2
    Module.Index0Float0Write=true
    Module.Index0Float1Write=true
    Module.Index0Float2Write=true
    Module.Index0Float3Write=true
   ) {module=DynamicMatParams}
   &ScaleColor(
    Scale Alpha=&Multiply_Float(
     A=&Multiply_Float(
      A=&FloatFromCurve(FloatCurve=DataInterface() {propertyValues {Curve {keys [
           {interpMode=RCIM_Cubic, tangentMode=RCTM_User, time=0, value=0}
           {interpMode=RCIM_Cubic, tangentMode=RCTM_User, time=0.5, value=1}
           {tangentMode=RCTM_User, time=1, value=0}
          ]}}})
      B=&Multiply_Float(
       A=&FNTimeOfDayScalar(Night Scale=0.012, Morning Scale=0.012, Day Scale=0.012, Evening Scale=0.01)
       B=&Multiply_Float(A=$System.SpawnInAlpha, B=&Multiply_Float(A=$System.EmoteAudioSyncCurve, B=$System.Waveform.Waveform))
      )
     )
     B=/Niagara/DynamicInputs/Helpers/Waveform("Waveform [1]"=::NewEnumerator0, "[1] Frequency"=3, "[1] Amplitude Min/Max"=<1 1.5>)
    )
    Scale RGB=&MakeVectorFromLinearColorRGB(Color=&MultiplyLinearColor(
      Linear Color=&LerpLinearColors(
       StartColor=<rgba 0.6 0.8333 1 1>
       EndColor=<rgba 0.48 0.828 1.2 1>
       LerpFactor=&RemapRange(Input Value To Remap=$System.AudioSyncAlpha, Input Min=0.5, Input Max=2, Output Min=0, Output Max=1)
      )
      Scale Factor=&MakeVector4_FromLinearColor(LinearColor=&FNTimeOfDayColor(Night Color=<rgba 0.6 0.7868 1 1>, Morning Color=<rgba 1 0.7111 0.7 1>, Day Color=<rgba 1 1 1 1>, Evening Color=<rgba 1 0.4788 0.45 1>))
     ))
    ScaleA=true
    ScaleRGB=true
   )
  ]
  rendererClasses [&NiagaraSpriteRendererProperties]
  renderers [
   {rendererClass=&NiagaraSpriteRendererProperties}
   {rendererClass=&NiagaraSpriteRendererProperties, rendererIndex=1}
  ]
 }
 SpotLight {
  emitterUpdateScript [
   &EmitterState(
    Inactive Response=::NewEnumerator1
    Life Cycle Mode=::NewEnumerator1
    Loop Behavior=::NewEnumerator1
    Loop Duration Mode=::NewEnumerator1
    UseLoopDelay=false
   )
   &SpawnBurst_Instantaneous(Use Spawn Probability=false, UseLoopCountLimit=false)
  ]
  particleSpawnScript [&InitializeParticle(
    Color Mode=::NewEnumerator0
    Lifetime=5
    Mesh Scale Mode=::NewEnumerator3
    Mesh Uniform Scale=0.5
    Sprite Size=<10 20>
    Sprite Size Mode=::NewEnumerator4
    UsePositionOffset=false
    Write Material Random=true
   )]
  particleUpdateScript [
   &ParticleState(Kill Particles When Lifetime Has Elapsed=false)
   &SocketLocation(
    Orientation Sampling=::NewEnumerator1
    SamplePosOffset=true
    Socket Index=9
    Socket Reader=$System.Socket Reader
    Socket Sampling Mode=::NewEnumerator3
    VelocitySpeedLimit=false
   )
   &ScaleColor(
    Scale Alpha=$System.AudioSyncAlpha
    Scale RGB=&MakeVectorFromLinearColorRGB(Color=&MultiplyLinearColor(
      Linear Color=<rgba 0.215 0.5918 1 0.8>
      Scale Factor=&MakeVector4_FromLinearColor(LinearColor=$System.TODColorMultiplier)
     ))
    ScaleA=true
    ScaleRGB=true
   )
   &FN_Light_Attributes(
    Create Radius Attribute=true
    Radius=&Waveform("Waveform [1]"=::NewEnumerator0, "[1] Frequency"=3, "[1] Amplitude Min/Max"=<50 80>)
   )
  ]
  rendererClasses [&NiagaraLightRendererProperties]
  renderers [{rendererClass=&NiagaraLightRendererProperties, rendererIndex=0}]
 }
 Spotlight_Flare {
  emitterUpdateScript [
   &EmitterState(
    Inactive Response=::NewEnumerator1
    Life Cycle Mode=::NewEnumerator1
    Loop Behavior=::NewEnumerator1
    Loop Duration Mode=::NewEnumerator1
    UseLoopDelay=false
   )
   &SpawnBurst_Instantaneous(Spawn Count=9, Use Spawn Probability=false, UseLoopCountLimit=false)
  ]
  particleSpawnScript [
   &PartitionParticles(Number Of Partitions=9)
   &InitializeParticle(
    Color=<rgba 0.3217 0.6382 1 0.15>
    Lifetime Max=0.6
    Lifetime Min=0.8
    Lifetime Mode=::NewEnumerator1
    Mesh Scale=<0.13 0.05 0.05>
    Mesh Scale Mode=::NewEnumerator4
    Sprite Size Mode=::NewEnumerator3
    Uniform Sprite Size=77
    UsePositionOffset=false
    Write Material Random=true
   )
   &InitialMeshOrientation(
    Mesh Orientation Mode=::NewEnumerator6
    Rotation=&MergeHelper_VectorSwitch(
     0=<0 -0.25 0>
     1=<0 -0.125 0>
     2=<0 0 0>
     3=<0 0.125 0>
     4=<0 0.2127 0>
     5=<0 0.2987 0>
     6=<0 0.416 0>
     7=<0 0.5333 0>
     8=<0 0.656 0>
     Index=$Particles.PartitionParticles.Partition
     Num=::NewEnumerator7
    )
    Use Rotation Vector=true
   )
   /SetVariables(Particles.InnerGlowColor:LinearColor=<rgba 0.215 0.5918 1 0.8>, Particles.InnerGlowScale:Vector3f=<0.05 0.05 0.04>) {isSetParameters=true}
  ]
  particleUpdateScript [
   &ParticleState(Kill Particles When Lifetime Has Elapsed=false, Let Infinitely Lived Particles Die When Emitter Deactivates=true)
   &ScaleMeshSize(Scale Factor=&Lerp_Vector(
     A=<0.8 1 0.8>
     B=<1.2 1.5 1.2>
     Alpha=&RemapRange(Input Value To Remap=$System.AudioSyncAlpha, Input Min=0.5, Input Max=2, Output Min=0, Output Max=1)
    ))
   &SocketLocation(
    Orientation Sampling=::NewEnumerator1
    SamplePosOffset=false
    Socket Index=$Particles.PartitionParticles.Partition
    Socket Reader=$System.Socket Reader
    Socket Sampling Mode=::NewEnumerator3
    VelocitySpeedLimit=false
   ) {module=SocketLocation001}
   &ScaleColor(
    Scale Alpha=$System.AudioSyncAlpha
    Scale RGB=&MakeVectorFromLinearColorRGB(Color=$System.TODColorMultiplier)
    ScaleA=true
    ScaleRGB=true
   )
   /SetVariables(
    Particles.InnerGlowColor:LinearColor=&MakeLinearColorFromVectorAndFloat(
     "Vector (RGB)"=&MakeVectorFromLinearColorRGB(Color=&MultiplyLinearColor(
       Linear Color=<rgba 0.215 0.5918 1 0.2>
       Scale Factor=&MakeVector4_FromLinearColor(LinearColor=$System.TODColorMultiplier)
      ))
     "Float (Alpha)"=$System.AudioSyncAlpha
    )
    Particles.InnerGlowScale:Vector3f=&Multiply_Vector(
     A=$Particles.Initial.InnerGlowScale
     B=&Lerp_Vector(
      A=<0.8 1 0.8>
      B=<1.2 1.5 1.2>
      Alpha=&RemapRange(Input Value To Remap=$System.AudioSyncAlpha, Input Min=0.5, Input Max=2, Output Min=0, Output Max=1)
     )
    )
   ) {isSetParameters=true}
   &FN_Sprite_Attributes(
    Camera Offset=1
    Color=&MakeLinearColorFromVectorAndFloat(
     "Vector (RGB)"=&MakeVectorFromLinearColorRGB(Color=&MultiplyLinearColor(
       Linear Color=<rgba 0.2258 0.7767 1 1>
       Scale Factor=&MakeVector4_FromLinearColor(LinearColor=$System.TODColorMultiplier)
      ))
     "Float (Alpha)"=&Multiply_Float(
      A=$System.AudioSyncAlpha
      B=&MakeFloatFromLinearColor(LinearColor=$System.TODColorMultiplier, Channel:ENiagaraFloatFromLinearColorOptions=::NewEnumerator3)
     )
    )
    Create Camera Offset Attribute=true
    Create Color Attribute=true
    Create SpriteSize Attribute=true
    SpriteSize=&Vector2DFromFloat(Value=&Lerp_Float(
      A=5
      B=7
      Alpha=&RemapRange(Input Value To Remap=$System.AudioSyncAlpha, Input Min=0.5, Input Max=2, Output Min=0, Output Max=1)
     ))
    Write to Module Namespace=true
   )
   &FN_Sprite_Attributes(
    Camera Offset=3
    Color=&MultiplyLinearColor(
     Linear Color=&MakeLinearColorFromVectorAndFloat(
      "Vector (RGB)"=&MakeVectorFromLinearColorRGB(Color=&MultiplyLinearColor(
        Linear Color=<rgba 0.2258 0.7767 1 1>
        Scale Factor=&MakeVector4_FromLinearColor(LinearColor=$System.TODColorMultiplier)
       ))
      "Float (Alpha)"=&Multiply_Float(
       A=$System.AudioSyncAlpha
       B=&MakeFloatFromLinearColor(LinearColor=$System.TODColorMultiplier, Channel:ENiagaraFloatFromLinearColorOptions=::NewEnumerator3)
      )
     )
     Scale Factor=<1 1 1 0.1>
    )
    Create Camera Offset Attribute=true
    Create Color Attribute=true
    Create SpriteSize Attribute=true
    SpriteSize=&Multiply_Vector2D_ByFloat(Float=1.5, Vector2D=$Particles.FN_Sprite_Attributes.SpriteSize)
    Write to Module Namespace=true
   ) {module=FN_Sprite_Attributes001}
  ]
  rendererClasses [&NiagaraMeshRendererProperties, &NiagaraSpriteRendererProperties]
  renderers [
   {rendererClass=&NiagaraMeshRendererProperties, rendererIndex=0}
   {rendererClass=&NiagaraMeshRendererProperties, rendererIndex=1}
   {rendererClass=&NiagaraSpriteRendererProperties, rendererIndex=2}
   {rendererClass=&NiagaraSpriteRendererProperties, rendererIndex=3}
  ]
 }
}
summary {
 systemName=NS_Emote_ClaimReflect_Barn_Lights
 emitters [
  {bEnabled=false, emitterName=NE_Debug_Float002, rendererClasses [&NiagaraSpriteRendererProperties]}
  {
   emitterName=Spotlight_Flare
   rendererClasses [&NiagaraMeshRendererProperties, &NiagaraSpriteRendererProperties]
  }
  {emitterName=SpotLight, rendererClasses [&NiagaraLightRendererProperties]}
  {bEnabled=false, emitterName=NE_Debug_Float003, rendererClasses [&NiagaraSpriteRendererProperties]}
  {emitterName=Smoke_Ground, rendererClasses [&NiagaraSpriteRendererProperties]}
 ]
 userVariables [
  User.bIsFrontEnd {type=NiagaraBool}
  User.bIsFrontEndPreview {type=NiagaraBool}
  User.ClaimReflect_AudioSync {type=NiagaraFloat, defaultValue=0, description=""}
 ]
}
